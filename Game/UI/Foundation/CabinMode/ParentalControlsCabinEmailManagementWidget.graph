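# Cabined-account parental controls widget, derived from FortCabinEmailManagementWidget, for viewing, changing or resending the parent or guardian email used to complete account setup.
# The edit-mode event swaps the email text and Change/Resend buttons with the input field and Done/Cancel buttons.
# The screen-state event toggles the sending and sent overlays, hiding both for initial and failed states.
# The failure event shows the passed error text in the red-tinted result line.
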WidgetBlueprint /Game/UI/Foundation/CabinMode/ParentalControlsCabinEmailManagementWidget
graphs [EventGraph {nodes [
   n1 = Event OnChangeEmailEditMode!() -> (OutputDelegate, then, bIsInEditMode:bool) {is=title K2Node_Event}
   > SetVisibility(
    self=[
     Get Button_Done() -> (Button_Done:Common Button Base Object Reference)
     Get Button_Cancel() -> (Button_Cancel:Common Button Base Object Reference)
     Get InputField_Email() -> (InputField_Email:Fort Input Field Object Reference)
    ]
    InVisibility=Select(Option 0=Collapsed, Option 1=SelfHitTestInvisible, Index=@n1.bIsInEditMode)
   )
   > SetVisibility(
    self=[
     Get Text_Email() -> (Text_Email:Common Text Object Reference)
     Get Button_EditEmail() -> (Button_EditEmail:Fort CTAButton Object Reference)
     Get Button_ResendEmail() -> (Button_ResendEmail:Fort CTAButton Object Reference)
    ]
    InVisibility=Select(Option 0=SelfHitTestInvisible, Option 1=Collapsed, Index=@n1.bIsInEditMode)
   )
   n2 = Event OnScreenStateChange!() -> (OutputDelegate, then, NewState:ECabinEmailViewState Enum=Initial) {is=title K2Node_Event}
   > n3 = Switch on ECabinEmailViewState!(execute, Selection:ECabinEmailViewState Enum=@n2.NewState) -> (Initial:Exec, Sending:Exec, Sent:Exec, SendFailed:Exec) {is=title K2Node_SwitchEnum}
   > SetVisibility(
    self=[
     Get Overlay_SendingEmail() -> (Overlay_SendingEmail:Overlay Object Reference)
     Get Overlay_SentEmail() -> (Overlay_SentEmail:Overlay Object Reference)
    ]
    InVisibility=Collapsed
   )
   SetVisibility(execute=@n3.SendFailed, self=[Get Overlay_SentEmail(), Get Overlay_SendingEmail()], InVisibility=Collapsed)
   SetVisibility(execute=@n3.Sending, self=Get Overlay_SentEmail(), InVisibility=Collapsed)
   > SetVisibility(self=Get Overlay_SendingEmail(), InVisibility=HitTestInvisible)
   SetVisibility(execute=@n3.Sent, self=Get Overlay_SentEmail(), InVisibility=HitTestInvisible)
   > SetVisibility(self=Get Overlay_SendingEmail(), InVisibility=Collapsed)
   n4 = Event OnShowFailureReason!() -> (OutputDelegate, then, ErrorText:"Text (by ref)") {is=title K2Node_Event}
   n5 = Get Text_EmailResult() -> (Text_EmailResult:Common Text Object Reference)
   > "SetText (Text)"(self=@n5, InText=@n4.ErrorText)
   > SetVisibility(self=@n5, InVisibility=HitTestInvisible)
  ]}]
variables []
widgets {
 info {inheritedWidgetCount=9, parentClass=/Script/FortniteUI.FortCabinEmailManagementWidget, rootWidgetClass=&Overlay, widgetCount=35}
 widgets [
  CabinEmailManagement {widgetClassPath=&Overlay, bIsVariable=true, widget=CabinEmailManagement}
  VerticalBox_6 {
   widgetClassPath=&VerticalBox
   parent=CabinEmailManagement
   widget=VerticalBox_6
   slot {class=OverlaySlot, properties {horizontalAlignment=HAlign_Fill, verticalAlignment=VAlign_Center}}
  }
  Border_0 {
   widgetClassPath=&Border
   parent=VerticalBox_6
   widget=Border_0
   properties {
    background {
     bIsDynamicallyLoaded=false
     drawAs=NoDrawType
     imageType=NoImage
     mirroring=NoMirror
     resourceName=None
     resourceObject=null
     tiling=NoTile
     imageSize=<32 32>
     margin=<ltrb 0 0 0 0>
     outlineSettings {
      bUseBrushTransparency=false
      roundingType=HalfHeightRadius
      width=0
      color {colorUseRule=UseColor_Specified, specifiedColor=<rgba 0 0 0 0>}
      cornerRadii=<0 0 0 0>
     }
     tintColor {colorUseRule=UseColor_Specified, specifiedColor=<rgba 1 1 1 1>}
     uVRegion {bIsValid=false, max=<0 0>, min=<0 0>}
    }
    contentColorAndOpacity=<rgba 0.6038 1 1 1>
    padding=<ltrb 0 0 0 0>
   }
   slot {class=VerticalBoxSlot}
  }
  Text_MainDescription {
   widgetClassPath=&CommonTextBlock
   bInherited=true
   parent=Border_0
   widget=Text_MainDescription
   properties {
    autoWrapText=true
    lineHeightPercentage=1.29
    mobileFontSizeMultiplier=1.3
    style=/FortUILibrary/Styles/Typography/_TextStyle_Body1._TextStyle_Body1_C
    text="To access restricted features and parental controls, your parent or guardian needs to complete your account setup. Please ask them to check their email."
    font {
     bForceMonospaced=false
     bMaterialIsStencil=false
     fontMaterial=null
     fontObject=&BurbankSmall
     letterSpacing=0
     monospacedWidth=1
     size=18
     skewAmount=0
     typefaceFontName=Bold
     outlineSettings {
      bApplyOutlineToDropShadows=false
      bMiteredCorners=false
      bSeparateFillAlpha=false
      outlineMaterial=null
      outlineSize=0
      outlineColor=<rgba 0 0 0 1>
     }
    }
    shadowOffset=<0 0>
    strikeBrush {
     bIsDynamicallyLoaded=false
     drawAs=NoDrawType
     imageType=NoImage
     mirroring=NoMirror
     resourceName=None
     resourceObject=null
     tiling=NoTile
     imageSize=<32 32>
     margin=<ltrb 0 0 0 0>
     outlineSettings {
      bUseBrushTransparency=false
      roundingType=HalfHeightRadius
      width=0
      color {colorUseRule=UseColor_Specified, specifiedColor=<rgba 0 0 0 0>}
      cornerRadii=<0 0 0 0>
     }
     tintColor {colorUseRule=UseColor_Specified, specifiedColor=<rgba 1 1 1 1>}
     uVRegion {bIsValid=false, max=<0 0>, min=<0 0>}
    }
   }
   slot {class=BorderSlot, properties {padding=<ltrb 0 0 0 0>}}
  }
  Spacer {widgetClassPath=&Spacer, parent=VerticalBox_6, widget=Spacer, properties {size=<1 48>}, slot {class=VerticalBoxSlot}}
  VerticalBox_1 {widgetClassPath=&VerticalBox, parent=VerticalBox_6, widget=VerticalBox_1, slot {class=VerticalBoxSlot}}
  ParentEmailWrapper {widgetClassPath=&HorizontalBox, bIsVariable=true, parent=VerticalBox_1, widget=ParentEmailWrapper, slot {class=VerticalBoxSlot}}
  Border_1 {
   widgetClassPath=&Border
   parent=ParentEmailWrapper
   widget=Border_1
   properties {
    background {
     bIsDynamicallyLoaded=false
     drawAs=NoDrawType
     imageType=NoImage
     mirroring=NoMirror
     resourceName=None
     resourceObject=null
     tiling=NoTile
     imageSize=<32 32>
     margin=<ltrb 0 0 0 0>
     outlineSettings {
      bUseBrushTransparency=false
      roundingType=HalfHeightRadius
      width=0
      color {colorUseRule=UseColor_Specified, specifiedColor=<rgba 0 0 0 0>}
      cornerRadii=<0 0 0 0>
     }
     tintColor {colorUseRule=UseColor_Specified, specifiedColor=<rgba 1 1 1 1>}
     uVRegion {bIsValid=false, max=<0 0>, min=<0 0>}
    }
    contentColorAndOpacity=<rgba 0.6038 1 1 1>
    padding=<ltrb 0 0 0 0>
   }
   slot {class=HorizontalBoxSlot, properties {verticalAlignment=VAlign_Center, padding=<ltrb 0 0 24 0>}}
  }
  Text_ParentEmail {
   widgetClassPath=&CommonTextBlock
   parent=Border_1
   widget=Text_ParentEmail
   properties {
    bIsScrollingEnabled=false
    lineHeightPercentage=1.29
    mobileFontSizeMultiplier=1.1
    style=&_TextStyle_Body1_C
    text=YOUR PARENT'S EMAIL:
    font {
     bForceMonospaced=false
     bMaterialIsStencil=false
     fontMaterial=null
     fontObject=&BurbankSmall
     letterSpacing=0
     monospacedWidth=1
     size=18
     skewAmount=0
     typefaceFontName=Bold
     outlineSettings {
      bApplyOutlineToDropShadows=false
      bMiteredCorners=false
      bSeparateFillAlpha=false
      outlineMaterial=null
      outlineSize=0
      outlineColor=<rgba 0 0 0 1>
     }
    }
    shadowOffset=<0 0>
    strikeBrush {
     bIsDynamicallyLoaded=false
     drawAs=NoDrawType
     imageType=NoImage
     mirroring=NoMirror
     resourceName=None
     resourceObject=null
     tiling=NoTile
     imageSize=<32 32>
     margin=<ltrb 0 0 0 0>
     outlineSettings {
      bUseBrushTransparency=false
      roundingType=HalfHeightRadius
      width=0
      color {colorUseRule=UseColor_Specified, specifiedColor=<rgba 0 0 0 0>}
      cornerRadii=<0 0 0 0>
     }
     tintColor {colorUseRule=UseColor_Specified, specifiedColor=<rgba 1 1 1 1>}
     uVRegion {bIsValid=false, max=<0 0>, min=<0 0>}
    }
   }
   slot {class=BorderSlot, properties {padding=<ltrb 0 0 0 0>}}
  }
  Text_Email {
   widgetClassPath=&CommonTextBlock
   bInherited=true
   parent=ParentEmailWrapper
   widget=Text_Email
   properties {
    bIsScrollingEnabled=false
    mobileFontSizeMultiplier=1.1
    text="[PH] placeholder email"
    textTransformPolicy=ToLower
    font {
     bForceMonospaced=false
     bMaterialIsStencil=false
     fontMaterial=null
     fontObject=&BurbankSmall
     letterSpacing=0
     monospacedWidth=1
     size=21
     skewAmount=0
     typefaceFontName=Black
     outlineSettings {
      bApplyOutlineToDropShadows=false
      bMiteredCorners=false
      bSeparateFillAlpha=false
      outlineMaterial=null
      outlineSize=0
      outlineColor=<rgba 0 0 0 1>
     }
    }
    shadowOffset=<0 0>
   }
   slot {class=HorizontalBoxSlot, properties {verticalAlignment=VAlign_Center}}
  }
  Overlay_EmailEntry_NotBind {widgetClassPath=&Overlay, bIsVariable=true, parent=VerticalBox_1, widget=Overlay_EmailEntry_NotBind, slot {class=VerticalBoxSlot}}
  Overlay_EmailEntry {
   widgetClassPath=&Overlay
   bInherited=true
   parent=Overlay_EmailEntry_NotBind
   widget=Overlay_EmailEntry
   slot {class=OverlaySlot, properties {horizontalAlignment=HAlign_Fill, verticalAlignment=VAlign_Fill}}
  }
  InputField_Email {
   widgetClassPath=/Game/UI/ContentControls/WBP_InputField
   bInherited=true
   bIsVariable=true
   parent=Overlay_EmailEntry
   widget=InputField_Email
   properties {
    bIsFocusable=true
    defaultHintTextValue=Enter Parent or Guardian email address...
    defaultLabelValue="Please enter your parent or guardian’s email address to continue."
    navigation=&self:WidgetTree.InputField_Email.WidgetNavigation_0
    showActionOnlyOnFocus=true
    showLabel=true
   }
   slot {class=OverlaySlot, properties {horizontalAlignment=HAlign_Fill}}
  }
  Spacer_1 {widgetClassPath=&Spacer, parent=VerticalBox_6, widget=Spacer_1, properties {size=<1 24>}, slot {class=VerticalBoxSlot}}
  Border {
   widgetClassPath=&Border
   parent=VerticalBox_6
   widget=Border
   properties {
    background {
     bIsDynamicallyLoaded=false
     drawAs=NoDrawType
     imageType=NoImage
     mirroring=NoMirror
     resourceName=None
     resourceObject=null
     tiling=NoTile
     imageSize=<32 32>
     margin=<ltrb 0 0 0 0>
     outlineSettings {
      bUseBrushTransparency=false
      roundingType=HalfHeightRadius
      width=0
      color {colorUseRule=UseColor_Specified, specifiedColor=<rgba 0 0 0 0>}
      cornerRadii=<0 0 0 0>
     }
     tintColor {colorUseRule=UseColor_Specified, specifiedColor=<rgba 1 1 1 1>}
     uVRegion {bIsValid=false, max=<0 0>, min=<0 0>}
    }
    contentColorAndOpacity=<rgba 0.6038 1 1 1>
    padding=<ltrb 0 0 0 24>
   }
   slot {class=VerticalBoxSlot}
  }
  Text_Warning {
   widgetClassPath=&CommonTextBlock
   bInherited=true
   parent=Border
   widget=Text_Warning
   properties {
    autoWrapText=true
    lineHeightPercentage=1.29
    mobileFontSizeMultiplier=1.3
    style=&_TextStyle_Body1_C
    text=Please confirm your parent's email is correct. You will not be able to change it for the next 24 hours.
    visibility=Collapsed
    font {
     bForceMonospaced=false
     bMaterialIsStencil=false
     fontMaterial=null
     fontObject=&BurbankSmall
     letterSpacing=0
     monospacedWidth=1
     size=18
     skewAmount=0
     typefaceFontName=Bold
     outlineSettings {
      bApplyOutlineToDropShadows=false
      bMiteredCorners=false
      bSeparateFillAlpha=false
      outlineMaterial=null
      outlineSize=0
      outlineColor=<rgba 0 0 0 1>
     }
    }
    shadowOffset=<0 0>
    strikeBrush {
     bIsDynamicallyLoaded=false
     drawAs=NoDrawType
     imageType=NoImage
     mirroring=NoMirror
     resourceName=None
     resourceObject=null
     tiling=NoTile
     imageSize=<32 32>
     margin=<ltrb 0 0 0 0>
     outlineSettings {
      bUseBrushTransparency=false
      roundingType=HalfHeightRadius
      width=0
      color {colorUseRule=UseColor_Specified, specifiedColor=<rgba 0 0 0 0>}
      cornerRadii=<0 0 0 0>
     }
     tintColor {colorUseRule=UseColor_Specified, specifiedColor=<rgba 1 1 1 1>}
     uVRegion {bIsValid=false, max=<0 0>, min=<0 0>}
    }
   }
   slot {class=BorderSlot, properties {padding=<ltrb 0 0 0 24>}}
  }
  HorizontalBox_0 {
   widgetClassPath=&HorizontalBox
   parent=VerticalBox_6
   widget=HorizontalBox_0
   slot {class=VerticalBoxSlot, properties {verticalAlignment=VAlign_Top, size {sizeRule=Fill, value=1}}}
  }
  Button_Done {
   widgetClassPath=&WBP_UIKit_Button_Regular
   bInherited=true
   bIsVariable=true
   parent=HorizontalBox_0
   widget=Button_Done
   properties {
    navigation=&self:WidgetTree.Button_Done.WidgetNavigation_0
    text=Done
    visibility=Collapsed
   }
   slot {class=HorizontalBoxSlot}
  }
  Button_Cancel {
   widgetClassPath=&WBP_UIKit_Button_Regular
   bInherited=true
   bIsVariable=true
   parent=HorizontalBox_0
   widget=Button_Cancel
   properties {
    navigation=&self:WidgetTree.Button_Cancel.WidgetNavigation_0
    text=Cancel
    visibility=Collapsed
   }
   slot {class=HorizontalBoxSlot, properties {padding=<ltrb 24 0 0 0>}}
  }
  Button_ResendEmail {
   widgetClassPath=&WBP_UIKit_Button_Regular
   bInherited=true
   bIsVariable=true
   parent=HorizontalBox_0
   widget=Button_ResendEmail
   properties {
    navigation=&self:WidgetTree.Button_ResendEmail.WidgetNavigation_0
    text=Resend Email
    hoveredSlateSoundOverride {resourceObject=/Game/Sounds/Fort_UI_Sounds/FrontEnd/New_FrontEnd/Fortnite_UI_MainMenu_Tab_Hover_01_Cue}
    pressedSlateSoundOverride {resourceObject=/Game/Sounds/Fort_UI_Sounds/FrontEnd/New_FrontEnd/Fortnite_UI_MainMenu_Generic_01_Cue}
   }
   slot {class=HorizontalBoxSlot}
  }
  Button_EditEmail {
   widgetClassPath=&WBP_UIKit_Button_Regular
   bInherited=true
   bIsVariable=true
   parent=HorizontalBox_0
   widget=Button_EditEmail
   properties {
    navigation=&self:WidgetTree.Button_EditEmail.WidgetNavigation_0
    text=Change Email
    hoveredSlateSoundOverride {resourceObject=&Fortnite_UI_MainMenu_Tab_Hover_01_Cue}
    pressedSlateSoundOverride {resourceObject=&Fortnite_UI_MainMenu_Generic_01_Cue}
   }
   slot {class=HorizontalBoxSlot, properties {padding=<ltrb 24 0 0 0>}}
  }
  Spacer_2 {widgetClassPath=&Spacer, parent=VerticalBox_6, widget=Spacer_2, properties {size=<1 16>}, slot {class=VerticalBoxSlot}}
  VerticalBox_0 {
   widgetClassPath=&VerticalBox
   parent=VerticalBox_6
   widget=VerticalBox_0
   slot {class=VerticalBoxSlot, properties {horizontalAlignment=HAlign_Left}}
  }
  Border_4 {
   widgetClassPath=&Border
   parent=VerticalBox_0
   widget=Border_4
   properties {
    visibility=HitTestInvisible
    background {
     bIsDynamicallyLoaded=false
     drawAs=NoDrawType
     imageType=NoImage
     mirroring=NoMirror
     resourceName=None
     resourceObject=null
     tiling=NoTile
     imageSize=<32 32>
     margin=<ltrb 0 0 0 0>
     outlineSettings {
      bUseBrushTransparency=false
      roundingType=HalfHeightRadius
      width=0
      color {colorUseRule=UseColor_Specified, specifiedColor=<rgba 0 0 0 0>}
      cornerRadii=<0 0 0 0>
     }
     tintColor {colorUseRule=UseColor_Specified, specifiedColor=<rgba 1 1 1 1>}
     uVRegion {bIsValid=false, max=<0 0>, min=<0 0>}
    }
    contentColorAndOpacity=<rgba 1 0.1714 0.1329 1>
    padding=<ltrb 0 0 0 0>
   }
   slot {class=VerticalBoxSlot}
  }
  Text_EmailResult {
   widgetClassPath=&CommonTextBlock
   bIsVariable=true
   parent=Border_4
   widget=Text_EmailResult
   properties {
    bAutoCollapseWithEmptyText=true
    bIsScrollingEnabled=false
    lineHeightPercentage=1.29
    mobileFontSizeMultiplier=1.2
    style=&_TextStyle_Body1_C
    font {
     bForceMonospaced=false
     bMaterialIsStencil=false
     fontMaterial=null
     fontObject=&BurbankSmall
     letterSpacing=0
     monospacedWidth=1
     size=18
     skewAmount=0
     typefaceFontName=Bold
     outlineSettings {
      bApplyOutlineToDropShadows=false
      bMiteredCorners=false
      bSeparateFillAlpha=false
      outlineMaterial=null
      outlineSize=0
      outlineColor=<rgba 0 0 0 1>
     }
    }
    shadowOffset=<0 0>
    strikeBrush {
     bIsDynamicallyLoaded=false
     drawAs=NoDrawType
     imageType=NoImage
     mirroring=NoMirror
     resourceName=None
     resourceObject=null
     tiling=NoTile
     imageSize=<32 32>
     margin=<ltrb 0 0 0 0>
     outlineSettings {
      bUseBrushTransparency=false
      roundingType=HalfHeightRadius
      width=0
      color {colorUseRule=UseColor_Specified, specifiedColor=<rgba 0 0 0 0>}
      cornerRadii=<0 0 0 0>
     }
     tintColor {colorUseRule=UseColor_Specified, specifiedColor=<rgba 1 1 1 1>}
     uVRegion {bIsValid=false, max=<0 0>, min=<0 0>}
    }
   }
   slot {class=BorderSlot, properties {padding=<ltrb 0 0 0 0>}}
  }
  Overlay_SendingEmail {widgetClassPath=&Overlay, bIsVariable=true, parent=VerticalBox_0, widget=Overlay_SendingEmail, slot {class=VerticalBoxSlot}}
  HorizontalBox_2 {
   widgetClassPath=&HorizontalBox
   parent=Overlay_SendingEmail
   widget=HorizontalBox_2
   slot {class=OverlaySlot, properties {horizontalAlignment=HAlign_Fill, verticalAlignment=VAlign_Center, padding=<ltrb 0 0 20 0>}}
  }
  Image_DisplayNameEntrySpinner {
   widgetClassPath=&Image
   bIsVariable=true
   parent=HorizontalBox_2
   widget=Image_DisplayNameEntrySpinner
   properties {brush {
     bIsDynamicallyLoaded=false
     drawAs=Image
     imageType=NoImage
     mirroring=NoMirror
     resourceName=None
     resourceObject=/FortUILibrary/UIKit/Throbber/MI_UIKit_Color_Throbber
     tiling=NoTile
     imageSize=<48 48>
     margin=<ltrb 0 0 0 0>
     outlineSettings {
      bUseBrushTransparency=false
      roundingType=HalfHeightRadius
      width=0
      color {colorUseRule=UseColor_Specified, specifiedColor=<rgba 0 0 0 0>}
      cornerRadii=<0 0 0 0>
     }
     tintColor {colorUseRule=UseColor_Specified, specifiedColor=<rgba 1 1 1 1>}
     uVRegion {bIsValid=false, max=<0 0>, min=<0 0>}
    }}
   slot {class=HorizontalBoxSlot, properties {horizontalAlignment=HAlign_Center, verticalAlignment=VAlign_Center, padding=<ltrb 6 0 6 0>}}
  }
  Border_2 {
   widgetClassPath=&Border
   parent=HorizontalBox_2
   widget=Border_2
   properties {
    background {
     bIsDynamicallyLoaded=false
     drawAs=NoDrawType
     imageType=NoImage
     mirroring=NoMirror
     resourceName=None
     resourceObject=null
     tiling=NoTile
     imageSize=<32 32>
     margin=<ltrb 0 0 0 0>
     outlineSettings {
      bUseBrushTransparency=false
      roundingType=HalfHeightRadius
      width=0
      color {colorUseRule=UseColor_Specified, specifiedColor=<rgba 0 0 0 0>}
      cornerRadii=<0 0 0 0>
     }
     tintColor {colorUseRule=UseColor_Specified, specifiedColor=<rgba 1 1 1 1>}
     uVRegion {bIsValid=false, max=<0 0>, min=<0 0>}
    }
    contentColorAndOpacity=<rgba 0.6038 1 1 1>
    padding=<ltrb 0 0 0 0>
   }
   slot {class=HorizontalBoxSlot, properties {verticalAlignment=VAlign_Center, padding=<ltrb 16 0 0 0>}}
  }
  Text_Email_1 {
   widgetClassPath=&CommonTextBlock
   parent=Border_2
   widget=Text_Email_1
   properties {
    lineHeightPercentage=1.29
    mobileFontSizeMultiplier=1.2
    style=&_TextStyle_Body1_C
    text=SENDING EMAIL...
    font {
     bForceMonospaced=false
     bMaterialIsStencil=false
     fontMaterial=null
     fontObject=&BurbankSmall
     letterSpacing=0
     monospacedWidth=1
     size=18
     skewAmount=0
     typefaceFontName=Bold
     outlineSettings {
      bApplyOutlineToDropShadows=false
      bMiteredCorners=false
      bSeparateFillAlpha=false
      outlineMaterial=null
      outlineSize=0
      outlineColor=<rgba 0 0 0 1>
     }
    }
    shadowOffset=<0 0>
    strikeBrush {
     bIsDynamicallyLoaded=false
     drawAs=NoDrawType
     imageType=NoImage
     mirroring=NoMirror
     resourceName=None
     resourceObject=null
     tiling=NoTile
     imageSize=<32 32>
     margin=<ltrb 0 0 0 0>
     outlineSettings {
      bUseBrushTransparency=false
      roundingType=HalfHeightRadius
      width=0
      color {colorUseRule=UseColor_Specified, specifiedColor=<rgba 0 0 0 0>}
      cornerRadii=<0 0 0 0>
     }
     tintColor {colorUseRule=UseColor_Specified, specifiedColor=<rgba 1 1 1 1>}
     uVRegion {bIsValid=false, max=<0 0>, min=<0 0>}
    }
   }
   slot {class=BorderSlot, properties {padding=<ltrb 0 0 0 0>}}
  }
  Overlay_SentEmail {widgetClassPath=&Overlay, bIsVariable=true, parent=VerticalBox_0, widget=Overlay_SentEmail, slot {class=VerticalBoxSlot}}
  HorizontalBox_3 {
   widgetClassPath=&HorizontalBox
   parent=Overlay_SentEmail
   widget=HorizontalBox_3
   slot {class=OverlaySlot, properties {horizontalAlignment=HAlign_Fill, verticalAlignment=VAlign_Center, padding=<ltrb 0 0 20 0>}}
  }
  Image_DisplayNameEntrySpinner_1 {
   widgetClassPath=&Image
   bIsVariable=true
   parent=HorizontalBox_3
   widget=Image_DisplayNameEntrySpinner_1
   properties {brush {
     bIsDynamicallyLoaded=false
     drawAs=Image
     imageType=NoImage
     mirroring=NoMirror
     resourceName=None
     resourceObject=/Game/Athena/HUD/Bacchus/HUDLayoutTool/HLT_Icons/T-HLT-Checkmark-64
     tiling=NoTile
     imageSize=<48 48>
     margin=<ltrb 0 0 0 0>
     outlineSettings {
      bUseBrushTransparency=false
      roundingType=HalfHeightRadius
      width=0
      color {colorUseRule=UseColor_Specified, specifiedColor=<rgba 0 0 0 0>}
      cornerRadii=<0 0 0 0>
     }
     tintColor {colorUseRule=UseColor_Specified, specifiedColor=<rgba 0.0369 1 0.2232 1>}
     uVRegion {bIsValid=false, max=<0 0>, min=<0 0>}
    }}
   slot {class=HorizontalBoxSlot, properties {horizontalAlignment=HAlign_Center, verticalAlignment=VAlign_Center, padding=<ltrb 6 0 6 0>}}
  }
  Border_3 {
   widgetClassPath=&Border
   parent=HorizontalBox_3
   widget=Border_3
   properties {
    background {
     bIsDynamicallyLoaded=false
     drawAs=NoDrawType
     imageType=NoImage
     mirroring=NoMirror
     resourceName=None
     resourceObject=null
     tiling=NoTile
     imageSize=<32 32>
     margin=<ltrb 0 0 0 0>
     outlineSettings {
      bUseBrushTransparency=false
      roundingType=HalfHeightRadius
      width=0
      color {colorUseRule=UseColor_Specified, specifiedColor=<rgba 0 0 0 0>}
      cornerRadii=<0 0 0 0>
     }
     tintColor {colorUseRule=UseColor_Specified, specifiedColor=<rgba 1 1 1 1>}
     uVRegion {bIsValid=false, max=<0 0>, min=<0 0>}
    }
    contentColorAndOpacity=<rgba 0.6038 1 1 1>
    padding=<ltrb 0 0 0 0>
   }
   slot {class=HorizontalBoxSlot, properties {verticalAlignment=VAlign_Center, padding=<ltrb 16 0 0 0>}}
  }
  Text_Email_2 {
   widgetClassPath=&CommonTextBlock
   parent=Border_3
   widget=Text_Email_2
   properties {
    lineHeightPercentage=1.29
    mobileFontSizeMultiplier=1.2
    style=&_TextStyle_Body1_C
    text="EMAIL SENT!"
    font {
     bForceMonospaced=false
     bMaterialIsStencil=false
     fontMaterial=null
     fontObject=&BurbankSmall
     letterSpacing=0
     monospacedWidth=1
     size=18
     skewAmount=0
     typefaceFontName=Bold
     outlineSettings {
      bApplyOutlineToDropShadows=false
      bMiteredCorners=false
      bSeparateFillAlpha=false
      outlineMaterial=null
      outlineSize=0
      outlineColor=<rgba 0 0 0 1>
     }
    }
    shadowOffset=<0 0>
    strikeBrush {
     bIsDynamicallyLoaded=false
     drawAs=NoDrawType
     imageType=NoImage
     mirroring=NoMirror
     resourceName=None
     resourceObject=null
     tiling=NoTile
     imageSize=<32 32>
     margin=<ltrb 0 0 0 0>
     outlineSettings {
      bUseBrushTransparency=false
      roundingType=HalfHeightRadius
      width=0
      color {colorUseRule=UseColor_Specified, specifiedColor=<rgba 0 0 0 0>}
      cornerRadii=<0 0 0 0>
     }
     tintColor {colorUseRule=UseColor_Specified, specifiedColor=<rgba 1 1 1 1>}
     uVRegion {bIsValid=false, max=<0 0>, min=<0 0>}
    }
   }
   slot {class=BorderSlot, properties {padding=<ltrb 0 0 0 0>}}
  }
 ]
}
